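# Post-style AnimBP for the 80s Exercise Commando glasses cosmetic, adding secondary motion to its hair, headband and glasses bones.
# Copies the parent mesh pose, then runs AnimDynamics with body-sphere limits on hair, constrains glasses to faceAttach, and applies corrective bone offsets.
# A near-duplicate dynamics chain is left unconnected to the output.
# EventGraph computes head-turn weights (front-end, non-emote) and a skydive/parachute flag; the AnimGraph shows no reads of them.
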AnimBlueprint /BRCosmetics/Accessories/Glasses/M_Med_Commando_80sExercise_Glasses/Meshes/M_Med_Commando_80sExercise_Glasses_AnimBP
target_skeleton=&/M_Med_Commando_80sExercise_Glasses_Skeleton
graphs [
 AnimGraph {nodes [
   n1 = "Anim Dynamics - Chain: dyn_80s_hair_01 -> dyn_80s_hair_02"(Alpha=1, ExternalForce="0,0,0") {
    angularDampingOverride=0.98
    angularSpringConstant=100
    bAngularSpring=true
    bChain=true
    bGravityOverrideInSimSpace=true
    bOverrideAngularDamping=true
    bUseSphericalLimits=true
    numSolverIterationsPostUpdate=4
    numSolverIterationsPreUpdate=16
    alphaBoolBlend {blendOption=Linear}
    boundBone {boneName=dyn_80s_hair_01}
    chainEnd {boneName=dyn_80s_hair_02}
    physicsBodyDefinitions [
     {
      collisionType=CustomSphere
      sphereCollisionRadius=1
      boundBone {boneName=dyn_80s_hair_01}
      boxExtents=<20 30 30>
      constraintSetup {angularTargetAxis=AxisZ, angularLimitsMax=<5 20 5>, angularLimitsMin=<-5 -20 -20>, angularTarget=<0 0 1>}
      localJointOffset=<2 0 0>
     }
     {
      boundBone {boneName=dyn_80s_hair_02}
      boxExtents=<10 10 10>
      constraintSetup {angularLimitsMax=<0 0 0>, angularLimitsMin=<0 0 0>, angularTarget=<0 0 0>}
      localJointOffset=<0 0 0>
     }
    ]
    sphericalLimits [{limitRadius=12, drivingBone {boneName=neck_01}, sphereLocalOffset=<2.5 6 0>}]
   }
   n2 = "Anim Dynamics - Chain: dyn_80s_hair_03 -> dyn_80s_hair_06"(ComponentPose=@n1, Alpha=1, ExternalForce="0,0,0") {
    angularDampingOverride=0.98
    angularSpringConstant=100
    bAngularSpring=true
    bChain=true
    bGravityOverrideInSimSpace=true
    bOverrideAngularDamping=true
    bUseSphericalLimits=true
    numSolverIterationsPostUpdate=4
    numSolverIterationsPreUpdate=16
    alphaBoolBlend {blendOption=Linear}
    boundBone {boneName=dyn_80s_hair_03}
    chainEnd {boneName=dyn_80s_hair_06}
    physicsBodyDefinitions [
     {
      collisionType=CustomSphere
      sphereCollisionRadius=1
      boundBone {boneName=dyn_80s_hair_03}
      boxExtents=<20 30 30>
      constraintSetup {angularTargetAxis=AxisZ, angularLimitsMax=<5 20 60>, angularLimitsMin=<-5 -20 -60>, angularTarget=<0 0 1>}
      localJointOffset=<2 0 0>
     }
     {
      boundBone {boneName=dyn_80s_hair_06}
      boxExtents=<10 10 10>
      constraintSetup {angularLimitsMax=<0 0 0>, angularLimitsMin=<0 0 0>, angularTarget=<0 0 0>}
      localJointOffset=<0 0 0>
     }
    ]
    sphericalLimits [{limitRadius=24, drivingBone {boneName=spine_05}, sphereLocalOffset=<-4 11 0>}]
   }
   n3 = "Anim Dynamics - Chain: dyn_80s_hair_04 -> dyn_80s_hair_05"(ComponentPose=@n2, Alpha=1, ExternalForce="0,0,0") {
    angularDampingOverride=0.98
    angularSpringConstant=100
    bAngularSpring=true
    bChain=true
    bGravityOverrideInSimSpace=true
    bOverrideAngularDamping=true
    bUseSphericalLimits=true
    numSolverIterationsPostUpdate=4
    numSolverIterationsPreUpdate=16
    alphaBoolBlend {blendOption=Linear}
    boundBone {boneName=dyn_80s_hair_04}
    chainEnd {boneName=dyn_80s_hair_05}
    physicsBodyDefinitions [
     {
      collisionType=CustomSphere
      sphereCollisionRadius=1
      boundBone {boneName=dyn_80s_hair_04}
      boxExtents=<20 30 30>
      constraintSetup {angularTargetAxis=AxisZ, angularLimitsMax=<5 20 60>, angularLimitsMin=<-5 -20 -60>, angularTarget=<0 0 1>}
      localJointOffset=<2 0 0>
     }
     {
      boundBone {boneName=dyn_80s_hair_05}
      boxExtents=<10 10 10>
      constraintSetup {angularLimitsMax=<0 0 0>, angularLimitsMin=<0 0 0>, angularTarget=<0 0 0>}
      localJointOffset=<0 0 0>
     }
    ]
    sphericalLimits [{limitRadius=24, drivingBone {boneName=spine_05}, sphereLocalOffset=<-4 11 0>}]
   }
   n4 = "Anim Dynamics - Chain: dyn_80s_hair_01_l -> dyn_80s_hair_02_l"(ComponentPose=@n3, Alpha=1, ExternalForce="0,0,0") {
    angularDampingOverride=0.98
    angularSpringConstant=100
    bAngularSpring=true
    bChain=true
    bGravityOverrideInSimSpace=true
    bOverrideAngularDamping=true
    bUseSphericalLimits=true
    numSolverIterationsPostUpdate=4
    numSolverIterationsPreUpdate=16
    alphaBoolBlend {blendOption=Linear}
    boundBone {boneName=dyn_80s_hair_01_l}
    chainEnd {boneName=dyn_80s_hair_02_l}
    physicsBodyDefinitions [
     {
      collisionType=CustomSphere
      sphereCollisionRadius=1.1
      boundBone {boneName=dyn_80s_hair_01_l}
      boxExtents=<20 30 30>
      constraintSetup {angularTargetAxis=AxisZ, angularLimitsMax=<5 20 5>, angularLimitsMin=<-5 -20 -20>, angularTarget=<0 0 1>}
      localJointOffset=<2 0 0>
     }
     {
      boundBone {boneName=dyn_80s_hair_02_l}
      boxExtents=<10 10 10>
      constraintSetup {angularLimitsMax=<0 0 0>, angularLimitsMin=<0 0 0>, angularTarget=<0 0 0>}
      localJointOffset=<0 0 0>
     }
    ]
    sphericalLimits [{limitRadius=24, drivingBone {boneName=spine_05}, sphereLocalOffset=<-4 11 -1>}]
   }
   n5 = "Anim Dynamics - Chain: dyn_80s_hair_01_r -> dyn_80s_hair_02_r"(ComponentPose=@n4, Alpha=1, ExternalForce="0,0,0") {
    angularDampingOverride=0.98
    angularSpringConstant=100
    bAngularSpring=true
    bChain=true
    bGravityOverrideInSimSpace=true
    bOverrideAngularDamping=true
    bUseSphericalLimits=true
    numSolverIterationsPostUpdate=4
    numSolverIterationsPreUpdate=16
    alphaBoolBlend {blendOption=Linear}
    boundBone {boneName=dyn_80s_hair_01_r}
    chainEnd {boneName=dyn_80s_hair_02_r}
    physicsBodyDefinitions [
     {
      collisionType=CustomSphere
      sphereCollisionRadius=1.1
      boundBone {boneName=dyn_80s_hair_01_r}
      boxExtents=<10 10 20>
      constraintSetup {angularTargetAxis=AxisZ, angularLimitsMax=<5 20 5>, angularLimitsMin=<-5 -20 -20>, angularTarget=<0 0 1>}
      localJointOffset=<-2 0 0>
     }
     {
      boundBone {boneName=dyn_80s_hair_02_r}
      boxExtents=<10 10 10>
      constraintSetup {angularLimitsMax=<0 0 0>, angularLimitsMin=<0 0 0>, angularTarget=<0 0 0>}
      localJointOffset=<0 0 0>
     }
    ]
    sphericalLimits [{limitRadius=24, drivingBone {boneName=spine_05}, sphereLocalOffset=<-4 11 1>}]
   }
   n6 = "Anim Dynamics - Chain: dyn_80s_hair_01_fr_l -> dyn_80s_hair_03_fr_l"(ComponentPose=@n5, Alpha=1, ExternalForce="0,0,0") {
    angularDampingOverride=0.999
    angularSpringConstant=500
    bAngularSpring=true
    bChain=true
    bGravityOverrideInSimSpace=true
    bOverrideAngularDamping=true
    bUseSphericalLimits=true
    numSolverIterationsPostUpdate=4
    numSolverIterationsPreUpdate=16
    alphaBoolBlend {blendOption=Linear}
    boundBone {boneName=dyn_80s_hair_01_fr_l}
    chainEnd {boneName=dyn_80s_hair_03_fr_l}
    physicsBodyDefinitions [
     {
      collisionType=CustomSphere
      sphereCollisionRadius=1.3
      boundBone {boneName=dyn_80s_hair_01_fr_l}
      boxExtents=<10 10 20>
      constraintSetup {angularTargetAxis=AxisZ, angularLimitsMax=<5 15 15>, angularLimitsMin=<-5 -15 -15>, angularTarget=<0 0 1>}
      localJointOffset=<2 0 0>
     }
     {
      boundBone {boneName=dyn_80s_hair_02_fr_l}
      boxExtents=<10 10 10>
      constraintSetup {angularLimitsMax=<0 0 0>, angularLimitsMin=<0 0 0>, angularTarget=<0 0 0>}
      localJointOffset=<0 0 0>
     }
     {
      boundBone {boneName=dyn_80s_hair_03_fr_l}
      boxExtents=<10 10 10>
      constraintSetup {angularLimitsMax=<0 0 0>, angularLimitsMin=<0 0 0>, angularTarget=<0 0 0>}
      localJointOffset=<0 0 0>
     }
    ]
    sphericalLimits [
     {limitRadius=20, drivingBone {boneName=spine_05}, sphereLocalOffset=<-5 -7 0>}
     {limitRadius=10, drivingBone {boneName=spine_05}, sphereLocalOffset=<9 2 -16>}
     {limitRadius=4, drivingBone {boneName=neck_02}, sphereLocalOffset=<0 2 0>}
    ]
   }
   n7 = "Anim Dynamics - Chain: dyn_80s_hair_01_fr_r -> dyn_80s_hair_03_fr_r"(ComponentPose=@n6, Alpha=1, ExternalForce="0,0,0") {
    angularDampingOverride=0.999
    angularSpringConstant=500
    bAngularSpring=true
    bChain=true
    bGravityOverrideInSimSpace=true
    bOverrideAngularDamping=true
    bUseSphericalLimits=true
    numSolverIterationsPostUpdate=4
    numSolverIterationsPreUpdate=16
    alphaBoolBlend {blendOption=Linear}
    boundBone {boneName=dyn_80s_hair_01_fr_r}
    chainEnd {boneName=dyn_80s_hair_03_fr_r}
    physicsBodyDefinitions [
     {
      collisionType=CustomSphere
      sphereCollisionRadius=2.1
      boundBone {boneName=dyn_80s_hair_01_fr_r}
      boxExtents=<10 10 20>
      constraintSetup {angularTargetAxis=AxisZ, angularLimitsMax=<5 20 10>, angularLimitsMin=<-5 -20 -10>, angularTarget=<0 0 1>}
      localJointOffset=<-2 0 0>
     }
     {
      boundBone {boneName=dyn_80s_hair_02_fr_r}
      boxExtents=<10 10 10>
      constraintSetup {angularLimitsMax=<0 0 0>, angularLimitsMin=<0 0 0>, angularTarget=<0 0 0>}
      localJointOffset=<0 0 0>
     }
     {
      boundBone {boneName=dyn_80s_hair_03_fr_r}
      boxExtents=<10 10 10>
      constraintSetup {angularLimitsMax=<0 0 0>, angularLimitsMin=<0 0 0>, angularTarget=<0 0 0>}
      localJointOffset=<0 0 0>
     }
    ]
    sphericalLimits [
     {limitRadius=20, drivingBone {boneName=spine_05}, sphereLocalOffset=<-5 -6 0>}
     {limitRadius=15, drivingBone {boneName=spine_05}, sphereLocalOffset=<2 4 19>}
     {limitRadius=4, drivingBone {boneName=neck_02}, sphereLocalOffset=<0 2 0>}
    ]
   }
   n8 = Anim Dynamics - Bone: dyn_80s_hair_top(ComponentPose=@n7, Alpha=0.1, ExternalForce="0,0,0") {
    bGravityOverrideInSimSpace=true
    bLinearSpring=true
    bOverrideLinearDamping=true
    gravityScale=0.1
    lODThreshold=1
    linearDampingOverride=0.999
    linearSpringConstant=320
    alphaBoolBlend {blendOption=Linear}
    boundBone {boneName=dyn_80s_hair_top}
    physicsBodyDefinitions [{
      sphereCollisionRadius=0
      boundBone {boneName=dyn_80s_hair_top}
      boxExtents=<100 100 100>
      constraintSetup {
       twistAxis=AxisZ
       angularLimitsMax=<0 0 0>
       angularLimitsMin=<0 0 0>
       angularTarget=<0 0 0>
       linearAxesMax=<10 10 10>
       linearAxesMin=<-10 -10 -10>
      }
      localJointOffset=<0 0 0>
     }]
   }
   n9 = Anim Dynamics - Bone: dyn_80s_hair_band_01(ComponentPose=@n8, Alpha=1, ExternalForce="0,0,0") {
    angularDampingOverride=0.99
    angularSpringConstant=3000
    bAngularSpring=true
    bGravityOverrideInSimSpace=true
    bOverrideAngularDamping=true
    numSolverIterationsPostUpdate=4
    numSolverIterationsPreUpdate=16
    alphaBoolBlend {blendOption=Linear}
    boundBone {boneName=dyn_80s_hair_band_01}
    physicsBodyDefinitions [{
      sphereCollisionRadius=0
      boundBone {boneName=dyn_80s_hair_band_01}
      boxExtents=<13 19 19>
      constraintSetup {angularLimitsMax=<5 20 0>, angularLimitsMin=<-5 -20 -20>, angularTarget=<1 0 0>}
      localJointOffset=<1.5 0 0>
     }]
   }
   n10 = Anim Dynamics - Bone: dyn_80s_hair_band_02(ComponentPose=@n9, Alpha=1, ExternalForce="0,0,0") {
    angularDampingOverride=0.99
    angularSpringConstant=3000
    bAngularSpring=true
    bGravityOverrideInSimSpace=true
    bOverrideAngularDamping=true
    numSolverIterationsPostUpdate=4
    numSolverIterationsPreUpdate=16
    alphaBoolBlend {blendOption=Linear}
    boundBone {boneName=dyn_80s_hair_band_02}
    physicsBodyDefinitions [{
      sphereCollisionRadius=0
      boundBone {boneName=dyn_80s_hair_band_02}
      boxExtents=<13 19 19>
      constraintSetup {angularLimitsMax=<5 20 0>, angularLimitsMin=<-5 -20 -20>, angularTarget=<1 0 0>}
      localJointOffset=<1.5 0 0>
     }]
   }
   n11 = Anim Dynamics - Bone: dyn_80s_hair_band_03(ComponentPose=@n10, Alpha=1, ExternalForce="0,0,0") {
    angularDampingOverride=0.99
    angularSpringConstant=3000
    bAngularSpring=true
    bGravityOverrideInSimSpace=true
    bOverrideAngularDamping=true
    numSolverIterationsPostUpdate=4
    numSolverIterationsPreUpdate=16
    alphaBoolBlend {blendOption=Linear}
    boundBone {boneName=dyn_80s_hair_band_03}
    physicsBodyDefinitions [{
      sphereCollisionRadius=0
      boundBone {boneName=dyn_80s_hair_band_03}
      boxExtents=<13 19 19>
      constraintSetup {angularLimitsMax=<5 20 0>, angularLimitsMin=<-5 -20 -20>, angularTarget=<1 0 0>}
      localJointOffset=<1.5 0 0>
     }]
   }
   n12 = Anim Dynamics - Bone: dyn_80s_hair_band_04(ComponentPose=@n11, Alpha=1, ExternalForce="0,0,0") {
    angularDampingOverride=0.99
    angularSpringConstant=3000
    bAngularSpring=true
    bGravityOverrideInSimSpace=true
    bOverrideAngularDamping=true
    numSolverIterationsPostUpdate=4
    numSolverIterationsPreUpdate=16
    alphaBoolBlend {blendOption=Linear}
    boundBone {boneName=dyn_80s_hair_band_04}
    physicsBodyDefinitions [{
      sphereCollisionRadius=0
      boundBone {boneName=dyn_80s_hair_band_04}
      boxExtents=<13 19 19>
      constraintSetup {angularLimitsMax=<5 20 0>, angularLimitsMin=<-5 -20 -20>, angularTarget=<1 0 0>}
      localJointOffset=<1.5 0 0>
     }]
   }
   n13 = Anim Dynamics - Bone: dyn_80s_hair_band_05(ComponentPose=@n12, Alpha=1, ExternalForce="0,0,0") {
    angularDampingOverride=0.99
    angularSpringConstant=3000
    bAngularSpring=true
    bGravityOverrideInSimSpace=true
    bOverrideAngularDamping=true
    numSolverIterationsPostUpdate=4
    numSolverIterationsPreUpdate=16
    alphaBoolBlend {blendOption=Linear}
    boundBone {boneName=dyn_80s_hair_band_05}
    physicsBodyDefinitions [{
      sphereCollisionRadius=0
      boundBone {boneName=dyn_80s_hair_band_05}
      boxExtents=<13 19 19>
      constraintSetup {angularLimitsMax=<5 20 0>, angularLimitsMin=<-5 -20 -20>, angularTarget=<1 0 0>}
      localJointOffset=<1.5 0 0>
     }]
   }
   n14 = Anim Dynamics - Bone: dyn_80s_hair_band_06(ComponentPose=@n13, Alpha=1, ExternalForce="0,0,0") {
    angularDampingOverride=0.99
    angularSpringConstant=3000
    bAngularSpring=true
    bGravityOverrideInSimSpace=true
    bOverrideAngularDamping=true
    numSolverIterationsPostUpdate=4
    numSolverIterationsPreUpdate=16
    alphaBoolBlend {blendOption=Linear}
    boundBone {boneName=dyn_80s_hair_band_06}
    physicsBodyDefinitions [{
      sphereCollisionRadius=0
      boundBone {boneName=dyn_80s_hair_band_06}
      boxExtents=<13 19 19>
      constraintSetup {angularLimitsMax=<5 20 0>, angularLimitsMin=<-5 -20 -20>, angularTarget=<1 0 0>}
      localJointOffset=<1.5 0 0>
     }]
   }
   n15 = Anim Dynamics - Bone: dyn_80s_glasses(ComponentPose=@n14, Alpha=1, ExternalForce="0,0,0") {
    angularDampingOverride=0.99
    angularSpringConstant=3000
    bAngularSpring=true
    bGravityOverrideInSimSpace=true
    bOverrideAngularDamping=true
    numSolverIterationsPostUpdate=4
    numSolverIterationsPreUpdate=16
    alphaBoolBlend {blendOption=Linear}
    boundBone {boneName=dyn_80s_glasses}
    physicsBodyDefinitions [{
      sphereCollisionRadius=0
      boundBone {boneName=dyn_80s_glasses}
      boxExtents=<5 20 10>
      constraintSetup {angularLimitsMax=<0 0 0>, angularLimitsMin=<0 0 -5>, angularTarget=<1 0 0>}
      localJointOffset=<0 10 0>
     }]
   }
   "Constraint - dyn_80s_glasses (P)"(ComponentPose=@n15, Alpha=1, ConstraintWeights_0=1) {
    alphaBoolBlend {blendOption=Linear}
    boneToModify {boneName=dyn_80s_glasses}
    constraintSetup [{targetBone {boneName=faceAttach}}]
    constraintWeights [1]
   }
   n16 = "Anim Dynamics - Chain: dyn_80s_hair_01 -> dyn_80s_hair_02"(ComponentPose=Local To Component(LocalPose=Copy Pose From Mesh() {bUseAttachedParent=true}), Alpha=1, ExternalForce="0,0,0") {
    angularDampingOverride=0.98
    angularSpringConstant=100
    bAngularSpring=true
    bChain=true
    bGravityOverrideInSimSpace=true
    bOverrideAngularDamping=true
    bUseSphericalLimits=true
    numSolverIterationsPostUpdate=4
    numSolverIterationsPreUpdate=16
    alphaBoolBlend {blendOption=Linear}
    boundBone {boneName=dyn_80s_hair_01}
    chainEnd {boneName=dyn_80s_hair_02}
    physicsBodyDefinitions [
     {
      collisionType=CustomSphere
      sphereCollisionRadius=1
      boundBone {boneName=dyn_80s_hair_01}
      boxExtents=<20 30 30>
      constraintSetup {angularTargetAxis=AxisZ, angularLimitsMax=<5 20 0>, angularLimitsMin=<-5 -20 -20>, angularTarget=<0 0 1>}
      localJointOffset=<2 0 0>
     }
     {
      boundBone {boneName=dyn_80s_hair_02}
      boxExtents=<10 10 10>
      constraintSetup {angularLimitsMax=<0 0 0>, angularLimitsMin=<0 0 0>, angularTarget=<0 0 0>}
      localJointOffset=<0 0 0>
     }
    ]
    sphericalLimits [{limitRadius=13, drivingBone {boneName=neck_01}, sphereLocalOffset=<2.5 6 0>}]
   }
   n17 = "Anim Dynamics - Chain: dyn_80s_hair_03 -> dyn_80s_hair_06"(ComponentPose=@n16, Alpha=1, ExternalForce="0,0,0") {
    angularDampingOverride=0.98
    angularSpringConstant=100
    bAngularSpring=true
    bChain=true
    bGravityOverrideInSimSpace=true
    bOverrideAngularDamping=true
    bUseSphericalLimits=true
    numSolverIterationsPostUpdate=4
    numSolverIterationsPreUpdate=16
    alphaBoolBlend {blendOption=Linear}
    boundBone {boneName=dyn_80s_hair_03}
    chainEnd {boneName=dyn_80s_hair_06}
    physicsBodyDefinitions [
     {
      collisionType=CustomSphere
      sphereCollisionRadius=1
      boundBone {boneName=dyn_80s_hair_03}
      boxExtents=<20 30 30>
      constraintSetup {angularTargetAxis=AxisZ, angularLimitsMax=<5 20 0>, angularLimitsMin=<-5 -20 -60>, angularTarget=<0 0 1>}
      localJointOffset=<2 0 0>
     }
     {
      boundBone {boneName=dyn_80s_hair_06}
      boxExtents=<10 10 10>
      constraintSetup {angularLimitsMax=<0 0 0>, angularLimitsMin=<0 0 0>, angularTarget=<0 0 0>}
      localJointOffset=<0 0 0>
     }
    ]
    sphericalLimits [{limitRadius=24, drivingBone {boneName=spine_05}, sphereLocalOffset=<-4 11 0>}]
   }
   n18 = "Anim Dynamics - Chain: dyn_80s_hair_04 -> dyn_80s_hair_05"(ComponentPose=@n17, Alpha=1, ExternalForce="0,0,0") {
    angularDampingOverride=0.98
    angularSpringConstant=100
    bAngularSpring=true
    bChain=true
    bGravityOverrideInSimSpace=true
    bOverrideAngularDamping=true
    bUseSphericalLimits=true
    numSolverIterationsPostUpdate=4
    numSolverIterationsPreUpdate=16
    alphaBoolBlend {blendOption=Linear}
    boundBone {boneName=dyn_80s_hair_04}
    chainEnd {boneName=dyn_80s_hair_05}
    physicsBodyDefinitions [
     {
      collisionType=CustomSphere
      sphereCollisionRadius=1
      boundBone {boneName=dyn_80s_hair_04}
      boxExtents=<20 30 30>
      constraintSetup {angularTargetAxis=AxisZ, angularLimitsMax=<5 20 0>, angularLimitsMin=<-5 -20 -60>, angularTarget=<0 0 1>}
      localJointOffset=<2 0 0>
     }
     {
      boundBone {boneName=dyn_80s_hair_05}
      boxExtents=<10 10 10>
      constraintSetup {angularLimitsMax=<0 0 0>, angularLimitsMin=<0 0 0>, angularTarget=<0 0 0>}
      localJointOffset=<0 0 0>
     }
    ]
    sphericalLimits [{limitRadius=24, drivingBone {boneName=spine_05}, sphereLocalOffset=<-4 11 0>}]
   }
   n19 = "Anim Dynamics - Chain: dyn_80s_hair_01_l -> dyn_80s_hair_02_l"(ComponentPose=@n18, Alpha=1, ExternalForce="0,0,0") {
    angularDampingOverride=0.98
    angularSpringConstant=100
    bAngularSpring=true
    bChain=true
    bGravityOverrideInSimSpace=true
    bOverrideAngularDamping=true
    bUseSphericalLimits=true
    numSolverIterationsPostUpdate=4
    numSolverIterationsPreUpdate=16
    alphaBoolBlend {blendOption=Linear}
    boundBone {boneName=dyn_80s_hair_01_l}
    chainEnd {boneName=dyn_80s_hair_02_l}
    physicsBodyDefinitions [
     {
      collisionType=CustomSphere
      sphereCollisionRadius=1.1
      boundBone {boneName=dyn_80s_hair_01_l}
      boxExtents=<20 30 30>
      constraintSetup {angularTargetAxis=AxisZ, angularLimitsMax=<5 20 5>, angularLimitsMin=<-5 -20 -20>, angularTarget=<0 0 1>}
      localJointOffset=<2 0 0>
     }
     {
      boundBone {boneName=dyn_80s_hair_02_l}
      boxExtents=<10 10 10>
      constraintSetup {angularLimitsMax=<0 0 0>, angularLimitsMin=<0 0 0>, angularTarget=<0 0 0>}
      localJointOffset=<0 0 0>
     }
    ]
    sphericalLimits [{limitRadius=24, drivingBone {boneName=spine_05}, sphereLocalOffset=<-4 11 -1>}]
   }
   n20 = "Anim Dynamics - Chain: dyn_80s_hair_01_r -> dyn_80s_hair_02_r"(ComponentPose=@n19, Alpha=1, ExternalForce="0,0,0") {
    angularDampingOverride=0.98
    angularSpringConstant=100
    bAngularSpring=true
    bChain=true
    bGravityOverrideInSimSpace=true
    bOverrideAngularDamping=true
    bUseSphericalLimits=true
    numSolverIterationsPostUpdate=4
    numSolverIterationsPreUpdate=16
    alphaBoolBlend {blendOption=Linear}
    boundBone {boneName=dyn_80s_hair_01_r}
    chainEnd {boneName=dyn_80s_hair_02_r}
    physicsBodyDefinitions [
     {
      collisionType=CustomSphere
      sphereCollisionRadius=1.1
      boundBone {boneName=dyn_80s_hair_01_r}
      boxExtents=<10 10 20>
      constraintSetup {angularTargetAxis=AxisZ, angularLimitsMax=<5 20 5>, angularLimitsMin=<-5 -20 -20>, angularTarget=<0 0 1>}
      localJointOffset=<-2 0 0>
     }
     {
      boundBone {boneName=dyn_80s_hair_02_r}
      boxExtents=<10 10 10>
      constraintSetup {angularLimitsMax=<0 0 0>, angularLimitsMin=<0 0 0>, angularTarget=<0 0 0>}
      localJointOffset=<0 0 0>
     }
    ]
    sphericalLimits [{limitRadius=24, drivingBone {boneName=spine_05}, sphereLocalOffset=<-4 11 1>}]
   }
   n21 = "Anim Dynamics - Chain: dyn_80s_hair_01_fr_l -> dyn_80s_hair_03_fr_l"(ComponentPose=@n20, Alpha=1, ExternalForce="0,0,0") {
    angularDampingOverride=0.999
    angularSpringConstant=500
    bAngularSpring=true
    bChain=true
    bGravityOverrideInSimSpace=true
    bOverrideAngularDamping=true
    bUseSphericalLimits=true
    numSolverIterationsPostUpdate=4
    numSolverIterationsPreUpdate=16
    alphaBoolBlend {blendOption=Linear}
    boundBone {boneName=dyn_80s_hair_01_fr_l}
    chainEnd {boneName=dyn_80s_hair_03_fr_l}
    physicsBodyDefinitions [
     {
      collisionType=CustomSphere
      sphereCollisionRadius=1.3
      boundBone {boneName=dyn_80s_hair_01_fr_l}
      boxExtents=<10 10 20>
      constraintSetup {angularTargetAxis=AxisZ, angularLimitsMax=<5 15 15>, angularLimitsMin=<-5 -15 -15>, angularTarget=<0 0 1>}
      localJointOffset=<2 0 0>
     }
     {
      boundBone {boneName=dyn_80s_hair_02_fr_l}
      boxExtents=<10 10 10>
      constraintSetup {angularLimitsMax=<0 0 0>, angularLimitsMin=<0 0 0>, angularTarget=<0 0 0>}
      localJointOffset=<0 0 0>
     }
     {
      boundBone {boneName=dyn_80s_hair_03_fr_l}
      boxExtents=<10 10 10>
      constraintSetup {angularLimitsMax=<0 0 0>, angularLimitsMin=<0 0 0>, angularTarget=<0 0 0>}
      localJointOffset=<0 0 0>
     }
    ]
    sphericalLimits [
     {limitRadius=20, drivingBone {boneName=spine_05}, sphereLocalOffset=<-5 -7 0>}
     {limitRadius=10, drivingBone {boneName=spine_05}, sphereLocalOffset=<9 2 -16>}
     {limitRadius=4, drivingBone {boneName=neck_02}, sphereLocalOffset=<0 2 0>}
    ]
   }
   n22 = "Anim Dynamics - Chain: dyn_80s_hair_01_fr_r -> dyn_80s_hair_03_fr_r"(ComponentPose=@n21, Alpha=1, ExternalForce="0,0,0") {
    angularDampingOverride=0.999
    angularSpringConstant=500
    bAngularSpring=true
    bChain=true
    bGravityOverrideInSimSpace=true
    bOverrideAngularDamping=true
    bUseSphericalLimits=true
    numSolverIterationsPostUpdate=4
    numSolverIterationsPreUpdate=16
    alphaBoolBlend {blendOption=Linear}
    boundBone {boneName=dyn_80s_hair_01_fr_r}
    chainEnd {boneName=dyn_80s_hair_03_fr_r}
    physicsBodyDefinitions [
     {
      collisionType=CustomSphere
      sphereCollisionRadius=2.1
      boundBone {boneName=dyn_80s_hair_01_fr_r}
      boxExtents=<10 10 20>
      constraintSetup {angularTargetAxis=AxisZ, angularLimitsMax=<5 20 10>, angularLimitsMin=<-5 -20 -10>, angularTarget=<0 0 1>}
      localJointOffset=<-2 0 0>
     }
     {
      boundBone {boneName=dyn_80s_hair_02_fr_r}
      boxExtents=<10 10 10>
      constraintSetup {angularLimitsMax=<0 0 0>, angularLimitsMin=<0 0 0>, angularTarget=<0 0 0>}
      localJointOffset=<0 0 0>
     }
     {
      boundBone {boneName=dyn_80s_hair_03_fr_r}
      boxExtents=<10 10 10>
      constraintSetup {angularLimitsMax=<0 0 0>, angularLimitsMin=<0 0 0>, angularTarget=<0 0 0>}
      localJointOffset=<0 0 0>
     }
    ]
    sphericalLimits [
     {limitRadius=20, drivingBone {boneName=spine_05}, sphereLocalOffset=<-5 -6 0>}
     {limitRadius=15, drivingBone {boneName=spine_05}, sphereLocalOffset=<2 4 19>}
     {limitRadius=4, drivingBone {boneName=neck_02}, sphereLocalOffset=<0 2 0>}
    ]
   }
   n23 = Anim Dynamics - Bone: dyn_80s_hair_top(ComponentPose=@n22, Alpha=0.1, ExternalForce="0,0,0") {
    bGravityOverrideInSimSpace=true
    bLinearSpring=true
    bOverrideLinearDamping=true
    gravityScale=0.1
    lODThreshold=1
    linearDampingOverride=0.999
    linearSpringConstant=320
    alphaBoolBlend {blendOption=Linear}
    boundBone {boneName=dyn_80s_hair_top}
    physicsBodyDefinitions [{
      sphereCollisionRadius=0
      boundBone {boneName=dyn_80s_hair_top}
      boxExtents=<100 100 100>
      constraintSetup {
       twistAxis=AxisZ
       angularLimitsMax=<0 0 0>
       angularLimitsMin=<0 0 0>
       angularTarget=<0 0 0>
       linearAxesMax=<10 10 10>
       linearAxesMin=<-10 -10 -10>
      }
      localJointOffset=<0 0 0>
     }]
   }
   n24 = Anim Dynamics - Bone: dyn_80s_hair_band_01(ComponentPose=@n23, Alpha=1, ExternalForce="0,0,0") {
    angularDampingOverride=0.99
    angularSpringConstant=3000
    bAngularSpring=true
    bGravityOverrideInSimSpace=true
    bOverrideAngularDamping=true
    numSolverIterationsPostUpdate=4
    numSolverIterationsPreUpdate=16
    alphaBoolBlend {blendOption=Linear}
    boundBone {boneName=dyn_80s_hair_band_01}
    physicsBodyDefinitions [{
      sphereCollisionRadius=0
      boundBone {boneName=dyn_80s_hair_band_01}
      boxExtents=<13 19 19>
      constraintSetup {angularLimitsMax=<5 20 0>, angularLimitsMin=<-5 -20 -20>, angularTarget=<1 0 0>}
      localJointOffset=<1.5 0 0>
     }]
   }
   n25 = Anim Dynamics - Bone: dyn_80s_hair_band_02(ComponentPose=@n24, Alpha=1, ExternalForce="0,0,0") {
    angularDampingOverride=0.99
    angularSpringConstant=3000
    bAngularSpring=true
    bGravityOverrideInSimSpace=true
    bOverrideAngularDamping=true
    numSolverIterationsPostUpdate=4
    numSolverIterationsPreUpdate=16
    alphaBoolBlend {blendOption=Linear}
    boundBone {boneName=dyn_80s_hair_band_02}
    physicsBodyDefinitions [{
      sphereCollisionRadius=0
      boundBone {boneName=dyn_80s_hair_band_02}
      boxExtents=<13 19 19>
      constraintSetup {angularLimitsMax=<5 20 0>, angularLimitsMin=<-5 -20 -20>, angularTarget=<1 0 0>}
      localJointOffset=<1.5 0 0>
     }]
   }
   n26 = Anim Dynamics - Bone: dyn_80s_hair_band_03(ComponentPose=@n25, Alpha=1, ExternalForce="0,0,0") {
    angularDampingOverride=0.99
    angularSpringConstant=3000
    bAngularSpring=true
    bGravityOverrideInSimSpace=true
    bOverrideAngularDamping=true
    numSolverIterationsPostUpdate=4
    numSolverIterationsPreUpdate=16
    alphaBoolBlend {blendOption=Linear}
    boundBone {boneName=dyn_80s_hair_band_03}
    physicsBodyDefinitions [{
      sphereCollisionRadius=0
      boundBone {boneName=dyn_80s_hair_band_03}
      boxExtents=<13 19 19>
      constraintSetup {angularLimitsMax=<5 20 0>, angularLimitsMin=<-5 -20 -20>, angularTarget=<1 0 0>}
      localJointOffset=<1.5 0 0>
     }]
   }
   n27 = Anim Dynamics - Bone: dyn_80s_hair_band_04(ComponentPose=@n26, Alpha=1, ExternalForce="0,0,0") {
    angularDampingOverride=0.99
    angularSpringConstant=3000
    bAngularSpring=true
    bGravityOverrideInSimSpace=true
    bOverrideAngularDamping=true
    numSolverIterationsPostUpdate=4
    numSolverIterationsPreUpdate=16
    alphaBoolBlend {blendOption=Linear}
    boundBone {boneName=dyn_80s_hair_band_04}
    physicsBodyDefinitions [{
      sphereCollisionRadius=0
      boundBone {boneName=dyn_80s_hair_band_04}
      boxExtents=<13 19 19>
      constraintSetup {angularLimitsMax=<5 20 0>, angularLimitsMin=<-5 -20 -20>, angularTarget=<1 0 0>}
      localJointOffset=<1.5 0 0>
     }]
   }
   n28 = Anim Dynamics - Bone: dyn_80s_hair_band_05(ComponentPose=@n27, Alpha=1, ExternalForce="0,0,0") {
    angularDampingOverride=0.99
    angularSpringConstant=3000
    bAngularSpring=true
    bGravityOverrideInSimSpace=true
    bOverrideAngularDamping=true
    numSolverIterationsPostUpdate=4
    numSolverIterationsPreUpdate=16
    alphaBoolBlend {blendOption=Linear}
    boundBone {boneName=dyn_80s_hair_band_05}
    physicsBodyDefinitions [{
      sphereCollisionRadius=0
      boundBone {boneName=dyn_80s_hair_band_05}
      boxExtents=<13 19 19>
      constraintSetup {angularLimitsMax=<5 20 0>, angularLimitsMin=<-5 -20 -20>, angularTarget=<1 0 0>}
      localJointOffset=<1.5 0 0>
     }]
   }
   n29 = Anim Dynamics - Bone: dyn_80s_hair_band_06(ComponentPose=@n28, Alpha=1, ExternalForce="0,0,0") {
    angularDampingOverride=0.99
    angularSpringConstant=3000
    bAngularSpring=true
    bGravityOverrideInSimSpace=true
    bOverrideAngularDamping=true
    numSolverIterationsPostUpdate=4
    numSolverIterationsPreUpdate=16
    alphaBoolBlend {blendOption=Linear}
    boundBone {boneName=dyn_80s_hair_band_06}
    physicsBodyDefinitions [{
      sphereCollisionRadius=0
      boundBone {boneName=dyn_80s_hair_band_06}
      boxExtents=<13 19 19>
      constraintSetup {angularLimitsMax=<5 20 0>, angularLimitsMin=<-5 -20 -20>, angularTarget=<1 0 0>}
      localJointOffset=<1.5 0 0>
     }]
   }
   n30 = Anim Dynamics - Bone: dyn_80s_glasses(ComponentPose=@n29, Alpha=1, ExternalForce="0,0,0") {
    angularDampingOverride=0.99
    angularSpringConstant=3000
    bAngularSpring=true
    bGravityOverrideInSimSpace=true
    bOverrideAngularDamping=true
    numSolverIterationsPostUpdate=4
    numSolverIterationsPreUpdate=16
    alphaBoolBlend {blendOption=Linear}
    boundBone {boneName=dyn_80s_glasses}
    physicsBodyDefinitions [{
      sphereCollisionRadius=0
      boundBone {boneName=dyn_80s_glasses}
      boxExtents=<5 20 10>
      constraintSetup {angularLimitsMax=<0 0 0>, angularLimitsMin=<0 0 -5>, angularTarget=<1 0 0>}
      localJointOffset=<0 10 0>
     }]
   }
   n31 = "Constraint - dyn_80s_glasses (P)"(ComponentPose=@n30, Alpha=1, ConstraintWeights_0=1) {
    alphaBoolBlend {blendOption=Linear}
    boneToModify {boneName=dyn_80s_glasses}
    constraintSetup [{targetBone {boneName=faceAttach}}]
    constraintWeights [1]
   }
   n32 = "Transform (Modify) Bone - Bone: dyn_80s_hair_01_fr_l"!(ComponentPose=@n31, bAlphaBoolEnabled:bool=True, Scale="1,0.75,0.75") -> (Pose) {
    alphaInputType=Bool
    rotationSpace=BCS_BoneSpace
    scaleMode=BMM_Additive
    scaleSpace=BCS_BoneSpace
    translationSpace=BCS_BoneSpace
    boneToModify {boneName=dyn_80s_hair_01_fr_l}
    scale {y=0.75, z=0.75}
   }
   n33 = "Transform (Modify) Bone - Bone: dyn_80s_hair_01_fr_l"!(ComponentPose=@n32, bAlphaBoolEnabled=True, Translation:Vector="-1,-1,1", Rotation:Rotator="2,2,0") -> (Pose) {
    alphaInputType=Bool
    rotationMode=BMM_Additive
    rotationSpace=BCS_BoneSpace
    translationMode=BMM_Additive
    translationSpace=BCS_BoneSpace
    boneToModify {boneName=dyn_80s_hair_01_fr_l}
    rotation {pitch=2, yaw=2}
    translation=<-1 -1 1>
   }
   n34 = "Transform (Modify) Bone - Bone: dyn_80s_hair_02_fr_l"!(ComponentPose=@n33, bAlphaBoolEnabled:bool=True, Translation="0,0,0", Rotation="-20,10,0") -> (Pose) {
    alphaInputType=Bool
    rotationMode=BMM_Additive
    rotationSpace=BCS_BoneSpace
    translationMode=BMM_Additive
    translationSpace=BCS_BoneSpace
    boneToModify {boneName=dyn_80s_hair_02_fr_l}
    rotation {pitch=-20, yaw=10}
   }
   n35 = "Transform (Modify) Bone - Bone: dyn_80s_hair_03_fr_l"!(ComponentPose=@n34, bAlphaBoolEnabled:bool=True, Translation="0,1,0", Rotation="10,0,0") -> (Pose) {
    alphaInputType=Bool
    rotationMode=BMM_Additive
    rotationSpace=BCS_BoneSpace
    translationMode=BMM_Additive
    translationSpace=BCS_BoneSpace
    boneToModify {boneName=dyn_80s_hair_03_fr_l}
    rotation {pitch=10}
    translation {y=1}
   }
   n36 = "Transform (Modify) Bone - Bone: dyn_80s_hair_01_fr_l"!(ComponentPose=@n35, Alpha:float=1, Translation="0,-1,-0.5", Rotation="2,2,0") -> (Pose) {
    rotationMode=BMM_Additive
    rotationSpace=BCS_BoneSpace
    translationMode=BMM_Additive
    translationSpace=BCS_BoneSpace
    boneToModify {boneName=dyn_80s_hair_01_fr_l}
    rotation {pitch=2, yaw=2}
    translation {y=-1, z=-0.5}
   }
   n37 = "Transform (Modify) Bone - Bone: dyn_80s_hair_02_fr_l"!(ComponentPose=@n36, Alpha:float=1, Translation="0,1.25,0.75", Rotation="0,0,0") -> (Pose) {
    rotationMode=BMM_Additive
    rotationSpace=BCS_BoneSpace
    translationMode=BMM_Additive
    translationSpace=BCS_BoneSpace
    boneToModify {boneName=dyn_80s_hair_02_fr_l}
    translation {y=1.25, z=0.75}
   }
   n38 = "Transform (Modify) Bone - Bone: dyn_80s_hair_02_fr_l"(ComponentPose=@n37, Alpha=1, Translation="0,0,0.5", Rotation="0,0,0") {
    rotationMode=BMM_Additive
    rotationSpace=BCS_ParentBoneSpace
    translationMode=BMM_Additive
    translationSpace=BCS_BoneSpace
    boneToModify {boneName=dyn_80s_hair_02_fr_l}
    translation {z=0.5}
   }
   n39 = "Transform (Modify) Bone - Bone: dyn_80s_hair_01"!(ComponentPose=@n38, bAlphaBoolEnabled:bool=True, Translation="0,0,0", Rotation="0,-30,0", Scale="1,1,1") -> (Pose) {
    alphaInputType=Bool
    rotationMode=BMM_Additive
    rotationSpace=BCS_BoneSpace
    alphaBoolBlend {blendInTime=0.1, blendOption=Linear, blendOutTime=0.5}
    boneToModify {boneName=dyn_80s_hair_01}
    rotation {yaw=-30}
   }
   n40 = "Transform (Modify) Bone - Bone: dyn_80s_hair_01_r"!(ComponentPose=@n39, bAlphaBoolEnabled:bool=True, Translation="0,0,0", Rotation="-10,-40,0", Scale="1,1,1") -> (Pose) {
    alphaInputType=Bool
    rotationMode=BMM_Additive
    rotationSpace=BCS_BoneSpace
    alphaBoolBlend {blendInTime=0.1, blendOption=Linear, blendOutTime=0.5}
    boneToModify {boneName=dyn_80s_hair_01_r}
    rotation {pitch=-10, yaw=-40}
   }
   n41 = "Transform (Modify) Bone - Bone: dyn_80s_hair_01_l"!(ComponentPose=@n40, bAlphaBoolEnabled:bool=True, Translation="0,0,0", Rotation="-20,-30,0", Scale="1,1,1") -> (Pose) {
    alphaInputType=Bool
    rotationMode=BMM_Additive
    rotationSpace=BCS_BoneSpace
    alphaBoolBlend {blendInTime=0.1, blendOption=Linear, blendOutTime=0.5}
    boneToModify {boneName=dyn_80s_hair_01_l}
    rotation {pitch=-20, yaw=-30}
   }
   n42 = "Transform (Modify) Bone - Bone: dyn_80s_hair_01"(ComponentPose=@n41, bAlphaBoolEnabled=True, Translation="0,0,0", Rotation="0,-15,0", Scale="1,1,1") {
    alphaInputType=Bool
    rotationMode=BMM_Additive
    rotationSpace=BCS_BoneSpace
    alphaBoolBlend {blendInTime=0.1, blendOption=Linear, blendOutTime=0.5}
    boneToModify {boneName=dyn_80s_hair_01}
    rotation {yaw=-15}
   }
   n43 = "Transform (Modify) Bone - Bone: dyn_80s_hair_01_r"(ComponentPose=@n42, bAlphaBoolEnabled=True, Translation="0,0,0", Rotation="-10,-20,0", Scale="1,1,1") {
    alphaInputType=Bool
    rotationMode=BMM_Additive
    rotationSpace=BCS_BoneSpace
    alphaBoolBlend {blendInTime=0.1, blendOption=Linear, blendOutTime=0.5}
    boneToModify {boneName=dyn_80s_hair_01_r}
    rotation {pitch=-10, yaw=-20}
   }
   n44 = "Transform (Modify) Bone - Bone: dyn_80s_hair_01_l"(ComponentPose=@n43, bAlphaBoolEnabled=True, Translation="0,0,0", Rotation="-15,-30,0", Scale="1,1,1") {
    alphaInputType=Bool
    rotationMode=BMM_Additive
    rotationSpace=BCS_BoneSpace
    alphaBoolBlend {blendInTime=0.1, blendOption=Linear, blendOutTime=0.5}
    boneToModify {boneName=dyn_80s_hair_01_l}
    rotation {pitch=-15, yaw=-30}
   }
   Output Pose(Result=Component To Local(ComponentPose=@n44)) {layerGroup=None}
  ]}
 EventGraph {nodes [
   Event BlueprintUpdateAnimation()
   > n1 = Sequence()
   n2 = Get bIsPlayingEmote()
   > Branch(Condition=AND Boolean(A=Get bIsInFrontEnd(), B=NOT Boolean(A=@n2)))
   n3 = GetAlphaFromReferenceTransformDelta(
    SkeletalMeshComponent=GetOwnerSkeletalMeshComponent()
    BoneName=head
    BaseName=neck_01
    DeltaComponentType=QuaternionTwist
    InputRange=(X=20,Y=35)
    TwistAxis="1,0,0"
   )
   > n4 = Set HeadLookingLeft(HeadLookingLeft:float=@n3) -> (Output_Get:float)
   > Set HeadLookingCenter(HeadLookingCenter:double=float - float(A=1, B=@n4)) -> (Output_Get:double)
   Set SkyDiveLogic(execute=@n1.then_1, SkyDiveLogic:bool=OR Boolean(A=Get bIsParachuteOpened(), B=Get bIsSkydiveDiveMode())) -> (Output_Get:bool)
   n5 = GetDeltaTransformFromRefPose!(self=GetOwnerSkeletalMeshComponent(), BoneName=neck_02, BaseName=spine_01) -> (ReturnValue_Location:Vector, ReturnValue_Rotation:Rotator, ReturnValue_Scale:Vector)
   BreakRotator(InRot=@n5.ReturnValue_Rotation)
   PrintString!(
    execute
    InString="To String (Rotator)"!(InRot:Rotator=@n5.ReturnValue_Rotation) -> (ReturnValue:String)
    bPrintToScreen
    bPrintToLog
    TextColor
    Duration
    Key
   ) -> (then)
   "To String (Float)"(InDouble=@n3)
  ]}
]
variables [
 HeadLookingCenter {type=number}
 HeadLookingLeft {type=number}
 SkyDiveLogic {type=boolean}
]
